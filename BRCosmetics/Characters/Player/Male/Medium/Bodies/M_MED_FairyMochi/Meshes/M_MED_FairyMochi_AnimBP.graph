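# Post-pose animation Blueprint for the FairyMochi medium male cosmetic body, copying the attached parent mesh's pose through a Control Rig.
# A bone-modifier chain then nudges the sword base, collapses its handle to zero scale, and adjusts forearm and upper-arm twist scales.
# Events set a hide-sword flag when the pickaxe has the FairyMochi tag and record cape backbling, though the visible AnimGraph reads neither flag.
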AnimBlueprint /BRCosmetics/Characters/Player/Male/Medium/Bodies/M_MED_FairyMochi/Meshes/M_MED_FairyMochi_AnimBP
target_skeleton=&/M_MED_FairyMochi_Skeleton
graphs [
 AnimGraph {nodes [
   n1 = Control Rig(Source=Copy Pose From Mesh() {bUseAttachedParent=true}) {controlRigAssetReference {blueprintRigClass=&/M_MED_FairyMochi_CtrlRig.M_MED_FairyMochi_CtrlRig_C}}
   n2 = "Transform (Modify) Bone - Bone: base_sword"(ComponentPose=Local To Component(LocalPose=@n1), Alpha=1, Translation="0,0.5,0", Rotation="0,0,0", Scale="1,1,1") {translationMode=BMM_Additive, translationSpace=BCS_BoneSpace, boneToModify {boneName=base_sword}, translation {y=0.5}}
   n3 = "Transform (Modify) Bone - Bone: handle_sword"(ComponentPose=@n2, Alpha=1, Translation="0,0,0", Rotation="0,0,0", Scale="0,0,0") {scaleMode=BMM_Replace, scaleSpace=BCS_BoneSpace, translationSpace=BCS_BoneSpace, boneToModify {boneName=handle_sword}, scale=<0 0 0>}
   n4 = "Transform (Modify) Bone - Bone: upperarm_twist_02_l"(
    ComponentPose="Transform (Modify) Bone - Bone: upperarm_twist_01_l"(
     ComponentPose="Transform (Modify) Bone - Bone: crv_lowerarm_l_top"(ComponentPose=@n3, Alpha=1, Translation="0,0,0", Rotation="0,0,0", Scale="1,1,2") {scaleMode=BMM_Additive, scaleSpace=BCS_BoneSpace, boneToModify {boneName=crv_lowerarm_l_top}, scale {z=2}}
     Alpha=1
     Translation="0,0,0"
     Rotation="0,0,0"
     Scale="1,1,1"
    ) {scaleMode=BMM_Additive, scaleSpace=BCS_BoneSpace, boneToModify {boneName=upperarm_twist_01_l}}
    Alpha=1
    Translation="0,0,0"
    Rotation="0,0,0"
    Scale="1,1,1.2"
   ) {scaleMode=BMM_Additive, scaleSpace=BCS_BoneSpace, boneToModify {boneName=upperarm_twist_02_l}, scale {z=1.2}}
   n5 = "Transform (Modify) Bone - Bone: upperarm_twist_02_r"(
    ComponentPose="Transform (Modify) Bone - Bone: upperarm_twist_01_r"(
     ComponentPose="Transform (Modify) Bone - Bone: crv_lowerarm_r_top"(ComponentPose=@n4, Alpha=1, Translation="0,0,0", Rotation="0,0,0", Scale="1,1,2") {scaleMode=BMM_Additive, scaleSpace=BCS_BoneSpace, boneToModify {boneName=crv_lowerarm_r_top}, scale {z=2}}
     Alpha=1
     Translation="0,0,0"
     Rotation="0,0,0"
     Scale="1,1,1"
    ) {scaleMode=BMM_Additive, scaleSpace=BCS_BoneSpace, boneToModify {boneName=upperarm_twist_01_r}}
    Alpha=1
    Translation="0,0,0"
    Rotation="0,0,0"
    Scale="1,0.745,1.2"
   ) {scaleMode=BMM_Additive, scaleSpace=BCS_BoneSpace, boneToModify {boneName=upperarm_twist_02_r}, scale {y=0.745, z=1.2}}
   Output Pose(Result=Component To Local(ComponentPose=@n5)) {layerGroup=None}
  ]}
 EventGraph {nodes [
   Event OnCharacterCustomizationLoaded()
   > n1 = Sequence()
   Set isWearingCape(execute=@n1.then_1, isWearingCape=CheckForBackbling().bIsBackblingACape)
   n2 = Branch(
    execute=[@n3.then, @n1.then_0]
    Condition=PickaxeItemDefinitionHasAnyMetaTag(Tags=Make Array("[0]"=(TagName="Cosmetics.MetaTags.FairyMochi")), bRequiresEquipped=true)
   )
   > Set hide sword(hide sword:bool=true) -> (Output_Get:bool)
   Set hide sword(execute=@n2.else, hide sword=false)
   n3 = Event OnWeaponChanged()
   Get Dynamic Material Instance()
   Event BlueprintUpdateAnimation()
  ]}
]
variables [
 hide sword {type=boolean}
 isWearingCape {type=boolean}
]
